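; Pawn Blueprint for the TidalCrane Jumper swarm NPC, a character that leaps forward when it spawns.
; After the parent BeginPlay and a short delay, it calls LaunchCharacter with both velocity overrides on.
; The launch velocity combines the actor's forward vector with an upward boost, each scaled by a ScalableFloat variable.
; SpawnActorOnDeath and the construction script only call their parent versions.
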(event EventBeginPlay
  ; Begin Play - Make these Jump Forward on Spawn
  (|CallParentFunction)
  (Utilities|FlowControl|Delay 0.2)
  (Character|LaunchCharacter (Utilities|Operators|Add (* (Transformation|GetActorForwardVector self) (ScalableFloat|GetValueatLevel (Variables|Default|GetJumpForwardDistance))) (Math|Vector|MakeVector 0.0 0.0 (ScalableFloat|GetValueatLevel (Variables|Default|GetJumpZAddition)))) true true))

(fn SpawnActorOnDeath
  (|CallParentFunction))

(fn ConstructionScript
  (|CallParentFunction))
